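# Masked, two-sided lit leaf particle material for the AncientLeer pickaxe effects, usable on Niagara sprites and mesh particles.
# An oak leaf texture, with optional center scale and offset on its UVs, drives a power-adjusted blend between two colors.
# That blend is multiplied by particle color and fed to an emissive adjustment function.
# The graph also holds leaf alpha, desaturated leaf color and a normal map sample.
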Material /BRCosmetics/Weapons/FORT_Melee/Pickaxe_AncientLeer/FX/M_AncientLeer_Leaf
graph {
 properties {
  bUsedWithNiagaraMeshParticles=true
  bUsedWithNiagaraSprites=true
  blendMode=BLEND_Masked
  pixelDepthOffsetMode=PDOM_Legacy
  twoSided=true
  usedShadingModels=Default Lit
 }
 nodes [
  n1 = VectorParameter() {parameterName=Color_01, defaultValue=<rgba 1 0.4735 0.0742 1>}
  n2 = VectorParameter() {parameterName=Color_02, defaultValue=<rgba 0.5647 0.0252 0.0262 1>}
  n3 = TextureCoordinate()
  n4 = VectorParameter() {parameterName=UV_Scale, defaultValue=<rgba 0.05 0.35 0 1>}
  n5 = MaterialFunctionCall(UVs=@n3, Texture Scale=AppendVector(A=@n4.R, B=@n4.R)) {materialFunction=&ScaleUVsByCenter}
  n6 = VectorParameter() {parameterName=UV_Offset, defaultValue=<rgba -2.5 0.35 0 1>}
  n7 = AppendVector(A=@n6.R, B=@n6.R)
  n8 = TextureSample(UVs=StaticSwitchParameter(True=Add(A=@n5.UVs, B=@n7), False=@n3) {parameterName=UVFix_Yes/No}) {samplerSource=SSM_Clamp_WorldGroupSettings, texture=/Game/Effects/Fort_Effects/World/Destruction/Textures/T_Oak_Tree_Leaf}
  n9 = Power(Base=@n8.RGB, Exp=ScalarParameter() {defaultValue=0.1, parameterName=Lerp_Power}) {constExponent=0.1}
  MaterialFunctionCall(Emissive=Multiply(A=ParticleColor().RGB, B=LinearInterpolate(A=@n1.RGB, B=@n2.RGB, Alpha=@n9))) {materialFunction=&MF_FX_EmissiveAdjustments}
  Reroute(None=@n8.A)
  Reroute(None=Desaturation(None=@n8.RGB, Fraction=Constant() {r=1}))
  TextureSample() {samplerType=SAMPLERTYPE_Normal, texture=/Game/Effects/Fort_Effects/World/Destruction/Textures/T_Oak_Tree_Leaf_N_01}
 ]
}
